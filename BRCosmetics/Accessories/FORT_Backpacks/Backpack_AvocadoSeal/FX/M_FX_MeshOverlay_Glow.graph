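# Additive unlit glow material for Niagara mesh particles, part of the Avocado Seal backpack cosmetic's FX.
# The V texture coordinate and a color parameter feed MF_FX_Refine, whose result is tinted by particle color and passed to MF_FX_EmissiveAdjustments, with no texture sampled in this graph.
# Separate branches scale depth-faded particle alpha by MainOpacity and raise V to a parameterized power.
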Material /BRCosmetics/Accessories/FORT_Backpacks/Backpack_AvocadoSeal/FX/M_FX_MeshOverlay_Glow
graph {
 properties {
  bAutomaticallySetUsageInEditor=false
  bUsedWithNiagaraMeshParticles=true
  blendMode=BLEND_Additive
  shadingModel=MSM_Unlit
  usedShadingModels=Unlit
 }
 nodes [
  n1 = ParticleColor()
  n2 = MaterialFunctionCall(Float2=TextureCoordinate()) {materialFunction=&BreakOutFloat2Components}
  n3 = VectorParameter() {defaultValue=<rgba 3 1 1 0.1>}
  MaterialFunctionCall(Emissive=Multiply(
    A=@n1.RGB
    B=MaterialFunctionCall(Texture=@n2.G, RGB=@n3.RGB, A=@n3.RGB) {materialFunction=&MF_FX_Refine}.Result
   )) {materialFunction=&MF_FX_EmissiveAdjustments}
  Multiply(
   A=DepthFade(Opacity=@n1.A, FadeDistance=ScalarParameter() {defaultValue=1, parameterName=FadeDistance})
   B=ScalarParameter() {defaultValue=3, parameterName=MainOpacity}
  )
  Power(Base=@n2.G, Exp=ScalarParameter() {defaultValue=6})
 ]
}
