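# Animation Blueprint for the AvocadoSeal backpack, which plays an elimination animation when its wearer eliminates an enemy.
# A bool blend selects front-end or gameplay state machine, both idling on that animation's first frame, then additive offsets shift the pack bone, with another during skydiving.
# An anim input event triggers gameplay playback through a briefly set flag.
# In the front end, the backpack part modifier's ReactivityEnabled starts delayed, repeating playback of the animation.
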AnimBlueprint /BRCosmetics/Accessories/FORT_Backpacks/Backpack_AvocadoSeal/Meshes/M_MED_AvocadoSeal_Pack_AnimBP
target_skeleton=&/M_MED_AvocadoSeal_Pack_Skeleton
graphs [
 AnimGraph {nodes [
   AnimGraphNode_StateMachine_0 = FE Idle + Reactive() {is=title AnimGraphNode_StateMachine}
   AnimGraphNode_StateMachine_3 = Backpack States() {is=title AnimGraphNode_StateMachine}
   n1 = Get bIsInFrontEnd()
   n2 = "Transform (Modify) Bone - Bone: pack"(
    ComponentPose=Local To Component(LocalPose=Blend Poses by bool(BlendPose_0=@AnimGraphNode_StateMachine_0, BlendPose_1=@AnimGraphNode_StateMachine_3, bActiveValue=@n1) {blendType=Linear})
    Translation="0,4.65,0"
    Rotation="0,0,0"
   ) {translationMode=BMM_Additive, alphaBoolBlend {blendOption=Linear}, boneToModify {boneName=pack}, translation {y=4.65}}
   n3 = "Transform (Modify) Bone - Bone: pack"!(ComponentPose=@n2, bAlphaBoolEnabled:bool=Get bIsSkyDiving(), Translation="0,-1.5,0", Rotation="0,0,0", Scale) -> (Pose) {alphaInputType=Bool, translationMode=BMM_Additive, alphaBoolBlend {blendOption=Linear}, boneToModify {boneName=pack}, translation {y=-1.5}}
   Output Pose(Result=Component To Local(ComponentPose=@n3))
  ]}
 FE Idle + Reactive {
  entry=FE_Idle
  path=AnimGraph.AnimGraphNode_StateMachine_0.FE Idle + Reactive
  ids=id
  nodes [
   AnimStateNode_0 = FE_Idle() {is=title AnimStateNode}
   AnimStateNode_1 = FE_Reactive() {is=title AnimStateNode, stateEntered {guid=#1, notifyName=Entered FE_Reactive_Anim}}
   AnimStateTransitionNode_0 = AnimStateTransitionNode()
   AnimStateTransitionNode_1 = AnimStateTransitionNode() {bAutomaticRuleBasedOnSequencePlayerInState=true}
  ]
  links [
   {
    dst=FE_Reactive
    rule=AnimGraph.AnimGraphNode_StateMachine_0.FE Idle + Reactive.AnimStateTransitionNode_0.Transition
    src=FE_Idle
    via=AnimStateTransitionNode_0
   }
   {
    dst=FE_Idle
    rule=AnimGraph.AnimGraphNode_StateMachine_0.FE Idle + Reactive.AnimStateTransitionNode_1.Transition
    src=FE_Reactive
    via=AnimStateTransitionNode_1
   }
  ]
 }
 FE_Idle {
  path=AnimGraph.AnimGraphNode_StateMachine_0.FE Idle + Reactive.AnimStateNode_0.FE_Idle
  nodes [Output Animation Pose(Result="Sequence Evaluator (by time) 'AvocadoSeal_Elimination_PACK'"!(ExplicitTime:float=0) -> (Pose:Pose Link Structure) {
     is=title AnimGraphNode_SequenceEvaluator
     bShouldLoop=false
     sequence=/BRCosmetics/Accessories/FORT_Backpacks/Backpack_AvocadoSeal/Animations/AvocadoSeal_Elimination_PACK
    })]
 }
 FE_Reactive {
  path=AnimGraph.AnimGraphNode_StateMachine_0.FE Idle + Reactive.AnimStateNode_1.FE_Reactive
  nodes [Output Animation Pose(Result=Sequence Player 'AvocadoSeal_Elimination_PACK'() {bLoopAnimation=false, sequence=&AvocadoSeal_Elimination_PACK})]
 }
 Transition {
  path=AnimGraph.AnimGraphNode_StateMachine_0.FE Idle + Reactive.AnimStateTransitionNode_0.Transition
  nodes [Result(bCanEnterTransition=Get Play Reactive FE Anim() -> (Play Reactive FE Anim:bool))]
 }
 Transition {path=AnimGraph.AnimGraphNode_StateMachine_0.FE Idle + Reactive.AnimStateTransitionNode_1.Transition, nodes [Result()]}
 Backpack States {
  entry=Idle
  path=AnimGraph.AnimGraphNode_StateMachine_3.Backpack States
  ids=id
  nodes [
   AnimStateNode_0 = Idle() {stateEntered {guid=#2}}
   AnimStateNode_4 = Elimination() {is=title AnimStateNode}
   AnimStateTransitionNode_11 = AnimStateTransitionNode() {crossfadeDuration=0}
   AnimStateTransitionNode_5 = AnimStateTransitionNode() {bAutomaticRuleBasedOnSequencePlayerInState=true, crossfadeDuration=0}
  ]
  links [
   {
    dst=Idle
    rule=AnimGraph.AnimGraphNode_StateMachine_3.Backpack States.AnimStateTransitionNode_5.Transition
    src=Elimination
    via=AnimStateTransitionNode_5
   }
   {
    dst=Elimination
    rule=AnimGraph.AnimGraphNode_StateMachine_3.Backpack States.AnimStateTransitionNode_11.Transition
    src=Idle
    via=AnimStateTransitionNode_11
   }
  ]
 }
 Idle {
  path=AnimGraph.AnimGraphNode_StateMachine_3.Backpack States.AnimStateNode_0.Idle
  nodes [Output Animation Pose(Result="Sequence Evaluator (by time) 'AvocadoSeal_Elimination_PACK'"(ExplicitTime=0) {bShouldLoop=false, sequence=&AvocadoSeal_Elimination_PACK})]
 }
 Elimination {
  path=AnimGraph.AnimGraphNode_StateMachine_3.Backpack States.AnimStateNode_4.Elimination
  nodes [Output Animation Pose(Result=Sequence Player 'AvocadoSeal_Elimination_PACK'() {bLoopAnimation=false, sequence=&AvocadoSeal_Elimination_PACK})]
 }
 Transition {
  path=AnimGraph.AnimGraphNode_StateMachine_3.Backpack States.AnimStateTransitionNode_11.Transition
  nodes [Result(bCanEnterTransition=Get bEnemyPlayerEliminated() -> (bEnemyPlayerEliminated:bool))]
 }
 Transition {path=AnimGraph.AnimGraphNode_StateMachine_3.Backpack States.AnimStateTransitionNode_5.Transition, nodes [Result()]}
 EventGraph {nodes [
   Event OnCharacterCustomizationLoaded()
   > n1 = Cast To FortPlayerPawn(Object=TryGetFortPawn())
   > n2 = Cast To B_Athena_PartModifier_Backpack_AvocadoSeal!(
    execute
    Object=GetModifierForPartType!(self:Fort Player Pawn Object Reference=@n1, PartType:EFortCustomPartType Enum=Backpack) -> (ReturnValue:Custom Character Part Modifier Object Reference)
   ) -> (then, CastFailed, AsB Athena Part Modifier Backpack Avocado Seal:B Athena Part Modifier Backpack Avocado Seal Object Reference) {is=title K2Node_DynamicCast}
   > n3 = Branch(Condition=Get ReactivityEnabled!(self:B Athena Part Modifier Backpack Avocado Seal Object Reference=@n2) -> (ReactivityEnabled:bool))
   > Do Once(Reset=@n5.then)
   > n4 = Delay(Duration=4)
   Set Play Reactive FE Anim(execute=[@n6.then, @n4.then], Play Reactive FE Anim:bool=true) -> (Output_Get:bool)
   n5 = Set Play Reactive FE Anim(execute=@n3.else, Play Reactive FE Anim=false)
   Event BlueprintPostEvaluateAnimation() {is=title K2Node_Event}
   > Set bEnemyPlayerEliminated(bEnemyPlayerEliminated:bool=false) -> (Output_Get:bool)
   AnimNotify_Entered FE_Reactive_Anim() {is=title K2Node_Event}
   > Set Play Reactive FE Anim(Play Reactive FE Anim=false)
   > n6 = Delay(Duration=10)
   n7 = Event BlueprintProcessAnimInputEvent()
   > Branch(Condition=IsNamed(self=@n7.AnimInputEvent, InEventName=aie.EnemyPlayerEliminated))
   > Set bEnemyPlayerEliminated(bEnemyPlayerEliminated=true)
   > Delay(Duration=0.2)
   > Set bEnemyPlayerEliminated(bEnemyPlayerEliminated=false)
   TryGetPawnOwner()
   Event BlueprintUpdateAnimation()
  ]}
]
variables [
 bEnemyPlayerEliminated {type=boolean}
 Play Reactive FE Anim {type=boolean}
]
