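# Physics asset for the AvocadoSeal backpack cosmetic mesh, giving the pack a simple collision shape.
# It defines a single body on the pack bone made of one tilted capsule offset from the bone origin.
# There are no constraints, so the asset describes one rigid volume with no jointed or ragdoll parts.
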PhysicsAsset /BRCosmetics/Accessories/FORT_Backpacks/Backpack_AvocadoSeal/Meshes/M_MED_AvocadoSeal_Pack_PhysicsAsset
bodies {pack [{shapeType=Capsule, length=16.4862, radius=15.1454, center=<-0.0002 -9.1121 -3.6476>, rotation=<rot -3.7245 -90 -89.9986>}]}
constraints []
